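# Idle glow Niagara system for the offhand BaskIsle pickaxe, drawing a warm light through sprite, mesh and light renderers.
# A single emitter burst-spawns a persistent particle that follows socket sampling on the system's skeletal mesh, with the light raised above it.
# System FadeIn scales the light and particle color.
# Particle brightness also follows a clamped camera-vector dot product, making the glow view-dependent; an older ScaleColor variant is disabled.
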NiagaraSystem /BRCosmetics/Weapons/FORT_Melee/Pickaxe_BaskIsle/FX/NS_Pickaxe_BaskIsle_Idle_Offhand
dependencies {
 usedDataInterfaces [
  &NiagaraDataInterfaceCamera
  &NiagaraDataInterfaceSkeletalMesh
  &NiagaraDataInterfaceSocketReader
  &NiagaraDataInterfaceCurve
 ]
 usedDynamicInputs [&FloatFromCurve, &Subtract_Float]
 usedModules [
  &self:SystemSpawnScript.SystemScriptSource.SystemScriptGraph.NiagaraNodeAssignment_1.SetVariables_#1
  &self:SystemSpawnScript.SystemScriptSource.SystemScriptGraph.NiagaraNodeAssignment_0.SetVariables_#2
  &self:SystemSpawnScript.SystemScriptSource.SystemScriptGraph.NiagaraNodeAssignment_3.SetVariables_#3
  &SystemState
 ]
}
emitters {Minimal {
  emitterUpdateScript [
   &EmitterState(Life Cycle Mode=::NewEnumerator1, Loop Behavior=::NewEnumerator1, Loop Duration Mode=::NewEnumerator1, UseLoopDelay=false)
   &SpawnBurst_Instantaneous(Use Spawn Probability=false, UseLoopCountLimit=false)
  ]
  particleSpawnScript [
   &InitializeParticle(
    Color=<rgba 1 0.6585 0.1261 0.558>
    Mesh Scale=<0.25 0.25 0.3>
    Mesh Scale Mode=::NewEnumerator4
    Sprite Size Mode=::NewEnumerator3
    Uniform Sprite Size=97.2387
    UsePositionOffset=false
    Write Material Random=true
   )
   &InitialMeshOrientation(Mesh Orientation Mode=::NewEnumerator7, Rotation=<0.5 0 0>, Use Rotation Vector=true)
  ]
  particleUpdateScript [
   &ParticleState(Kill Particles When Lifetime Has Elapsed=false)
   &SkeletalMeshLocation(MaskSpawnGroup=false, Mesh Sampling Type=::NewEnumerator5, Skeletal Mesh=$System.Skeletal Mesh, Socket Sampling Mode=::NewEnumerator5)
   &FN_Light_Attributes(
    Color=<rgba 1 0.6585 0.1261 0.558>
    Create Color Attribute=true
    Create Exponent Attribute=true
    Create Position Attribute=true
    Create Radius Attribute=true
    Exponent=2
    Position=&AddVectorToPosition(Position=$Particles.Position, Vector=<0 0 30>)
    Radius=30
    Scale Color=$System.FadeIn
   )
   &CameraOffset(Camera Offset Amount=-2)
   /SetVariables(
    Particles.ConeLightCameraOffset=-2
    Particles.ConeLightColor:LinearColor=<rgba 1 0.7142 0.2811 0.2>
    Particles.ConePosition:NiagaraPosition=&AddVectorToPosition(Position=$Particles.Position, Vector=<0 0 0>)
   ) {isSetParameters=true}
   /SetVariables(
    Particles.AirGlowDynamicParam:Vector4f=<4 0 0 0>
    Particles.AirGlowSpriteSize:Vector2f=&Vector2DFromFloat(Value=7.5)
   ) {isSetParameters=true}
   &ScaleColor(
    Scale RGB=&VectorFromFloat(Value=&Lerp_Float(
      A=0.4
      B=4
      Alpha=&ClampFloat(Float=&DotProduct(Vector A=&ReturnCameraVector(Camera Vectors=::NewEnumerator1, Coordinate Space=::Simulation), Vector B=<1 0 0>))
     ))
    ScaleA=true
    ScaleRGB=true
   ) {enabled=false, module=ScaleColor001}
   &ScaleColor(
    Scale RGB=&MakeVectorFromLinearColorRGB(Color=&Multiply_LinearColorByFloat(
      Linear Color=<rgba 1 1 1 1>
      Float=&MakeFloatFromVector4(
       Vector 4=&MakeVector4_FromFloat(Float=&Multiply_Float(
         A=&Multiply_Float(A=2.5, B=$System.FadeIn)
         B=&ClampFloat(
          Float=&DotProduct(
           Vector A=&ReturnCameraVector(Camera Vectors=::NewEnumerator1, Coordinate Space=::Local)
           Vector B=&FN_FrontEndVector3Modifier(Gameplay=<1 0 -1>, Front End=<1 0 -1>)
          )
          Max=0.5
         )
        ))
       Channel=::NewEnumerator0
      )
     ))
    ScaleA=true
    ScaleRGB=true
   )
  ]
  rendererClasses [
   &NiagaraSpriteRendererProperties
   &NiagaraLightRendererProperties
   &NiagaraMeshRendererProperties
  ]
  renderers [
   {rendererClass=&NiagaraSpriteRendererProperties}
   {rendererClass=&NiagaraLightRendererProperties, rendererIndex=1}
   {rendererClass=&NiagaraMeshRendererProperties, rendererIndex=2}
  ]
 }}
summary {
 systemName=NS_Pickaxe_BaskIsle_Idle_Offhand
 emitters [{
   emitterName=Minimal
   rendererClasses [
    &NiagaraSpriteRendererProperties
    &NiagaraLightRendererProperties
    &NiagaraMeshRendererProperties
   ]
  }]
 userVariables [
  User.bIsFrontEnd {type=NiagaraBool}
  User.bIsFrontEndPreview {type=NiagaraBool}
 ]
}
